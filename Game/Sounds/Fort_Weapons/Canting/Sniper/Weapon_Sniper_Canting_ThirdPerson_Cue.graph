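# Third-person sound cue for canting a sniper weapon, a foley sound.
# A single wave player feeds a modulator that holds its volume at a fixed reduced level.
# The cue applies a volume boost, the weapon reload attenuation asset and the weapon foley sound class, with a limited audible distance.
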SoundCue /Game/Sounds/Fort_Weapons/Canting/Sniper/Weapon_Sniper_Canting_ThirdPerson_Cue
root=Modulator_0
nodes [
 {id=Modulator_0, children [WavePlayer_1], properties {volumeMax=0.7, volumeMin=0.7}}
 {id=WavePlayer_1, properties {soundWaveAssetPtr=&/Weapon_Sniper_Canting_ThirdPerson_01}}
]
properties {
 attenuationSettings=&Fort_Weapon_Reload_Att
 maxDistance=2250
 soundClassObject=&Fort_WeaponFoley
 volumeMultiplier=2
}
